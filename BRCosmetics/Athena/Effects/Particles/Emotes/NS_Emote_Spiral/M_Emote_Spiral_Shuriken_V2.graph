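# Translucent, unlit, two-sided Niagara mesh-particle material for a spinning shuriken in the Spiral emote effect.
# A ring texture displaces the rotating shuriken UVs, and the sampled texture drives a curve-atlas blend between dark and light emissive colors.
# Opacity combines shuriken alpha, a radial-coordinate panning dissolve and an adjustable center hole.
# A sphere-mask ring adds index-of-refraction distortion, with dynamic parameters controlling displacement, rotation, refraction and dissolve.
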Material /BRCosmetics/Athena/Effects/Particles/Emotes/NS_Emote_Spiral/M_Emote_Spiral_Shuriken_V2
graph {
 properties {
  bUsedWithNiagaraMeshParticles=true
  blendMode=BLEND_Translucent
  pixelDepthOffsetMode=PDOM_Legacy
  refractionMethod=RM_IndexOfRefraction
  shadingModel=MSM_Unlit
  twoSided=true
  usedShadingModels=Unlit
 }
 nodes [
  n1 = ParticleColor()
  n2 = TextureCoordinate()
  n3 = Constant() {r=0.5}
  n4 = Subtract(
   A=SphereMask(A=@n2, B=@n3, Radius=Constant() {r=0.5}, Hardness=Constant())
   B=SphereMask(A=@n2, B=@n3, Radius=Constant() {r=0.35}, Hardness=Constant() {r=0.5})
  )
  n5 = DynamicParameter() {defaultValue=<rgba 0.2 1 1 1>, paramNames [Displace Intensity, Rotation Speed, Ref Intensity, Dissolve]}
  LinearInterpolate(Alpha=Multiply(A=@n1.A, B=Multiply(A=@n5.Ref Intensity, B=@n4))) {constA=1, constB=1.05}
  n6 = VectorParameter() {group=Color, parameterName=Color Dark, defaultValue=<rgba 0.4653 0.6435 1 1>}
  n7 = VectorParameter() {group=Color, parameterName=Color Light, defaultValue=<rgba 0.9774 0.9925 1 1>}
  n8 = TextureSample!(UVs, Tex, Level, Apply View MipBias) {
   mipValueMode=TMVM_MipLevel
   samplerSource=SSM_Clamp_WorldGroupSettings
   texture=&/T_Emote_Spiral_Ring
  }
  n9 = Multiply(A=@n8.G, B=@n5.Displace Intensity)
  n10 = Time()
  n11 = Rotator(
   Coordinate=TextureCoordinate()
   Time=LinearInterpolate(
    A=Multiply(A=@n5.Rotation Speed, B=@n10)
    B=DynamicParameter() {parameterIndex=1, paramNames [Dssolve Rotation, Dynamic Rotation, "", ""]}.Dynamic Rotation
    Alpha=ScalarParameter() {parameterName=Use Dynamic Rotation}
   )
  ) {speed=-10}
  n12 = TextureSampleParameter2D!(UVs=Add(A=TextureCoordinate() {uTiling=0, vTiling=0}, B=Add(A=@n9, B=@n11)), Level, Apply View MipBias) {
   mipValueMode=TMVM_MipLevel
   parameterName=Shuriken Tex
   samplerSource=SSM_Clamp_WorldGroupSettings
   texture=&/T_Emote_Spiral_Shuriken_02_V2
  }
  n13 = CurveAtlasRowParameter(CurveTime=@n12.G) {atlas=/Game/Athena/LinearColorCurves/A_Atlas_Turtleneck_Glider, curve=/Game/Athena/LinearColorCurves/Curve_Emote_Spiral, defaultValue=22}
  MaterialFunctionCall(Emissive=Multiply(A=LinearInterpolate(A=@n6.RGB, B=@n7.RGB, Alpha=@n13.B), B=@n1.RGB)) {materialFunction=&MF_FX_EmissiveAdjustments}
  n14 = MaterialFunctionCall() {materialFunction=&VectorToRadialValue}
  n15 = VectorParameter() {group=Dissolve, parameterName=Dissolve UV, defaultValue=<rgba 2 2 -6 0>}
  n16 = AppendVector(A=@n15.R, B=@n15.R)
  n17 = DynamicParameter() {parameterIndex=1, paramNames [Dssolve Rotation, Dynamic Rotation, "", ""]}
  n18 = TextureSampleParameter2D(UVs=Panner(Coordinate=Multiply(A=@n14.Radial Coordinates, B=@n16), Time=Multiply(A=@n15.B, B=@n17.Dssolve Rotation)) {speedX=1}) {
   group=Dissolve
   mipValueMode=TMVM_MipLevel
   parameterName=Dissolve Tex
   texture=/Game/Effects/Fort_Effects/Textures/Power_Source/T_Power_Source_Sphere_Texture
  }
  n19 = Power(Base=@n18.G, Exp=ScalarParameter() {defaultValue=1, group=Dissolve, parameterName=Dissolve Power})
  n20 = ScalarParameter() {defaultValue=5, group=Dissolve, parameterName=Dissolve Edge Hardness, sliderMin=0.5}
  n21 = DynamicParameter() {defaultValue=<rgba 0.2 1 1 1>, paramNames [Displace Intensity, Rotation Speed, Ref Intensity, Dissolve]}
  n22 = Multiply(A=ScalarParameter() {defaultValue=1, group=Dissolve, parameterName=Dissolve Dynamic Test}, B=@n21.Dissolve)
  n23 = Saturate(None=Subtract(A=Multiply(A=@n19, B=@n20), B=LinearInterpolate(A=@n20, Alpha=@n22) {constB=-1}))
  n24 = ChannelMaskParameter(None=@n12.RGBA) {group=Alpha, maskChannel=Alpha, parameterName=Alpha Channel, defaultValue=<rgba 0 0 0 1>}
  n25 = TextureCoordinate()
  n26 = Constant() {r=0.5}
  n27 = ScalarParameter() {group=Alpha, parameterName=Hole Size}
  n28 = ScalarParameter() {defaultValue=1, group=Alpha, parameterName=Hole Soft}
  Multiply(A=Subtract(A=Multiply(A=@n23, B=@n24.R), B=SphereMask(A=@n25, B=@n26, Radius=@n27, Hardness=@n28)), B=@n1.A)
  TextureSample!(UVs=TextureCoordinate() {uTiling=-1}, Tex, Apply View MipBias) {samplerType=SAMPLERTYPE_Normal, texture=&/T_Emote_Spiral_Normal_01}
 ]
}
