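# Translucent unlit Niagara mesh-particle material for the glowing sphere in the spiral emote effect.
# Facing-angle core glow and Fresnel rims, broken up by two panning swirl noises and V-gradient masks, drive core colors and a curve-atlas dark-to-light tint.
# Dynamic parameters control noise blend, FX Process color and intensity shifts, and normal-direction vertex displacement from panning noise.
# Particle color tints emissive; particle alpha and depth fade set opacity.
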Material /BRCosmetics/Athena/Effects/Particles/Emotes/NS_Emote_Spiral/M_Emote_Spiral_Sphere_V2
graph {
 properties {
  bUsedWithNiagaraMeshParticles=true
  blendMode=BLEND_Translucent
  pixelDepthOffsetMode=PDOM_Legacy
  shadingModel=MSM_Unlit
  usedShadingModels=Unlit
 }
 nodes [
  n1 = DotProduct(A=VertexNormalWS(), B=CameraVectorWS())
  n2 = Power(Base=@n1, Exp=ScalarParameter() {defaultValue=5, group=Fresnel, parameterName=Inside A Power}) {constExponent=20}
  n3 = ScalarParameter() {defaultValue=0.1, group=Fresnel, parameterName=Inside A Intensity}
  n4 = ScalarParameter() {defaultValue=30, group=Fresnel, parameterName=Inside B Power}
  n5 = ScalarParameter() {defaultValue=0.35, group=Fresnel, parameterName=Inside B Intensity 01}
  n6 = ScalarParameter() {defaultValue=0.75, group=Fresnel, parameterName=Inside B Intensity 02}
  n7 = DynamicParameter() {defaultValue=<rgba 3 1 1 1>, paramNames [WPO Intensity, Noise Lerp, Param3, FX Process]}
  n8 = Add(
   A=Saturate(None=Multiply(A=@n2, B=@n3))
   B=Multiply(A=Power(Base=@n1, Exp=@n4) {constExponent=20}, B=LinearInterpolate(A=@n5, B=@n6, Alpha=@n7.FX Process))
  )
  n9 = VectorParameter() {group=Noise, parameterName=Noise A UV, defaultValue=<rgba 2 3 3 4>}
  n10 = AppendVector(A=@n9.R, B=@n9.R)
  n11 = TextureCoordinate()
  n12 = TextureSampleParameter2D!(UVs=Panner(Coordinate=Multiply(A=@n11, B=@n10), Speed=AppendVector(A=@n9.B, B=@n9.B)), Level, Apply View MipBias) {group=Noise, mipValueMode=TMVM_MipLevel, parameterName=Noise A Tex, texture=/Game/Global/Textures/Default/Flowmaps/T_SwirlFlow_02_D}
  n13 = VectorParameter() {group=Noise, parameterName=Noise B UV, defaultValue=<rgba 2 2 2 2>}
  n14 = AppendVector(A=@n13.R, B=@n13.R)
  n15 = TextureSampleParameter2D(UVs=Panner(Coordinate=Multiply(A=@n11, B=@n14), Speed=AppendVector(A=@n13.B, B=@n13.B))) {
   group=Noise
   mipValueMode=TMVM_MipLevel
   parameterName=Noise B Tex
   samplerType=SAMPLERTYPE_LinearGrayscale
   texture=/Game/Global/Textures/Default/Noise/T_SwirlNoise_N_height
  }
  n16 = MaterialFunctionCall() {materialFunction=&LinearGradient}
  n17 = OneMinus(None=@n16.VGradient)
  n18 = Multiply(
   A=Multiply(A=LinearInterpolate(A=@n12.G, B=@n15.G) {constAlpha=0.35}, B=Multiply(A=@n16.VGradient, B=@n17))
   B=ScalarParameter() {defaultValue=30, group=Noise, parameterName=Noise Intensity}
  ) {constB=6}
  n19 = Multiply(A=@n8, B=@n18)
  n20 = Multiply(
   A=ScalarParameter() {defaultValue=1, group=Noise, parameterName=Lerp Noise}
   B=DynamicParameter() {defaultValue=<rgba 3 1 1 1>, paramNames [WPO Intensity, Noise Lerp, Param3, FX Process]}.Noise Lerp
  )
  n21 = MaterialFunctionCall() {materialFunction=&LinearGradient}
  n22 = Multiply(A=@n21.VGradient, B=OneMinus(None=@n21.VGradient))
  n23 = Multiply(A=Saturate(None=Subtract(A=@n22) {constB=0.248})) {constA=10, constB=4}
  n24 = Multiply(
   A=LinearInterpolate(
    A=VectorParameter() {parameterName=Color Core A, defaultValue=<rgba 0 3.3333 5 1>}.RGB
    B=VectorParameter() {parameterName=Color Core B, defaultValue=<rgba 0 3 0.25 1>}.RGB
    Alpha=DynamicParameter() {defaultValue=<rgba 3 1 1 1>, paramNames [WPO Intensity, Noise Lerp, Param3, FX Process]}.FX Process
   )
   B=Subtract(A=LinearInterpolate(A=@n8, B=@n19, Alpha=@n20), B=@n23)
  )
  n25 = Fresnel(
   ExponentIn=ScalarParameter() {defaultValue=10, parameterName=Fresnel A ExponentIn}
   BaseReflectFractionIn=ScalarParameter() {defaultValue=0.04, group=Fresnel, parameterName=Fresnel A Base}
  )
  n26 = ScalarParameter() {defaultValue=0.3, group=Fresnel, parameterName=Fresnel A Intensity}
  n27 = Multiply(
   A=Fresnel(
    ExponentIn=ScalarParameter() {defaultValue=2, parameterName=Fresnel B ExponentIn}
    BaseReflectFractionIn=ScalarParameter() {group=Fresnel, parameterName=Fresnel B Base}
   )
   B=LinearInterpolate(
    A=ScalarParameter() {defaultValue=0.05, group=Fresnel, parameterName=Fresnel B Intensity 01}
    B=ScalarParameter() {group=Fresnel, parameterName=Fresnel B Intensity 02}
    Alpha=@n7.FX Process
   )
  )
  n28 = Subtract(A=Add(A=Multiply(A=@n25, B=@n26), B=@n27), B=@n23)
  n29 = CurveAtlasRowParameter(CurveTime=LinearInterpolate(A=@n28, B=Multiply(A=@n28, B=@n18), Alpha=@n20)) {atlas=/Game/Athena/LinearColorCurves/A_Atlas_Turtleneck_Glider, curve=/Game/Athena/LinearColorCurves/Curve_Emote_Spiral, defaultValue=22}
  n30 = Add(
   A=LinearInterpolate(
    A=VectorParameter() {group=Color, parameterName=Color Dark, defaultValue=<rgba 0 0.0208 0.05 1>}.RGB
    B=VectorParameter() {group=Color, parameterName=Color Light, defaultValue=<rgba 0.25 1.0417 5 1>}.RGB
    Alpha=@n29.A
   )
   B=Multiply(A=@n27, B=VectorParameter() {group=Color, parameterName=Color Fresnel, defaultValue=<rgba 0 0.25 1 1>}.RGB)
  )
  n31 = ParticleColor()
  MaterialFunctionCall(Emissive=Multiply(A=Add(A=@n24, B=@n30), B=@n31.RGB)) {materialFunction=&MF_FX_EmissiveAdjustments}
  n32 = TextureSample(UVs=Panner(Coordinate=TextureCoordinate()) {speedX=1, speedY=2}) {texture=/Game/Characters/Player/Male/Medium/Bodies/M_MED_Viper/Textures/Glitch/T_SoftNoise}
  n33 = MaterialFunctionCall() {materialFunction=&LinearGradient}
  n34 = Multiply(A=Multiply(A=@n33.VGradient, B=OneMinus(None=@n33.VGradient))) {constB=2}
  Multiply(
   A=Multiply(A=Multiply(A=@n32.RGB, B=@n34), B=VertexNormalWS())
   B=DynamicParameter() {defaultValue=<rgba 3 1 1 1>, paramNames [WPO Intensity, Noise Lerp, Param3, FX Process]}.WPO Intensity
  ) {constB=3}
  Multiply(
   A=@n31.A
   B=DepthFade(
    Opacity=ScalarParameter() {defaultValue=1, group=Alpha, parameterName=Alpha}
    FadeDistance=ScalarParameter() {defaultValue=25, group=Alpha, parameterName=FadeDistance}
   )
  )
 ]
}
