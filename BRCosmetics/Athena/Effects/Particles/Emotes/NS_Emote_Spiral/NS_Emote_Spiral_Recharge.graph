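# Niagara system for the recharge phase of the Spiral emote, showing light-blue converging airflow that builds up at the character's right-hand weapon bone.
# A single mesh emitter samples weapon_r through a skeletal mesh interface and spawns short-lived particles at a rising rate.
# System and emitter age curves grow mesh size and opacity over time, while time of day scales alpha.
# Dynamic material parameters animate a sweep over each particle's life.
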NiagaraSystem /BRCosmetics/Athena/Effects/Particles/Emotes/NS_Emote_Spiral/NS_Emote_Spiral_Recharge
dependencies {
 usedDataInterfaces [&NiagaraDataInterfaceCurve, &NiagaraDataInterfaceSkeletalMesh]
 usedDynamicInputs []
 usedModules [&SystemState]
}
emitters {Convergent_airflow_start {
  emitterUpdateScript [
   &EmitterState(Life Cycle Mode=::NewEnumerator1, Loop Behavior=::NewEnumerator1, Loop Duration=1.8, UseLoopDelay=false)
   &SpawnRate(
    SpawnRate=&FloatFromCurve(FloatCurve=DataInterface() {propertyValues {Curve {keys [<at 0 0>,<at 0.4 7>,<at 0.8 20>]}}}, CurveIndex=$Emitter.NormalizedLoopAge)
    Use Spawn Probability=false
   )
  ]
  particleSpawnScript [
   /Niagara/Modules/Spawn/Initialization/InitializeParticle(
    Color=<rgba 0.4 0.7 1 1>
    Lifetime=&RandomRangeFloat(Minimum=0.5, Maximum=0.5)
    Mesh Scale=&Multiply_VectorByFloat(Vector=&RandomRangeVector(Minimum=<0.7 0.7 0.7>, Maximum=<1 1 1>), Float=0.55)
    Module.Write Color=true
    Module.Write Lifetime=true
    Module.Write Mass=true
    Module.Write Material Random=true
    Module.Write Position=false
    Module.Write Scale=true
    Module.Write SpriteRotation=false
    Module.Write SpriteSize=true
    Position=&SimulationPosition()
    Sprite Size=<5 5>
   )
   &SkeletalMeshLocation(
    Bone Sampling Mode=::NewEnumerator4
    Inherit Velocity=false
    Interpolate Mesh Sampling=false
    MaskSpawnGroup=false
    Sampled Position Offset=<-10 0 0>
    Sampled Position Offset Coordinate Space:ENiagaraMeshTransforms=::NewEnumerator3
    Skeletal Mesh=DataInterface() {propertyValues {
      DefaultMesh=null
      ExcludeBoneName=root
      OptionalComponentClass=null
      PreviewMesh=&M_Med_Soldier_04_Preview
      SkinningMode=SkinOnTheFly
      SoftSourceActor=null
      SourceMode=Default
      UvSetIndex=0
      WholeMeshLOD=-1
      bExcludeBone=false
      bReadDeformedGeometry=true
      bRequireCurrentFrameData=true
      ComponentTags []
      FilteredBones [weapon_r]
      FilteredSockets []
      MeshUserParameter {parameter {name=None, typeDefHandle {registeredTypeIndex=1}, varData []}}
      SamplingRegions []
     }}
   )
   &InitialMeshOrientation(Use Rotation Vector=false)
  ]
  particleUpdateScript [
   &ParticleState()
   &SolveForcesAndVelocity(Clamp Velocity=false, Limit Acceleration=false)
   &DynamicMaterialParameters(
    Index 0 Param 1=&FloatFromCurve(FloatCurve=DataInterface() {propertyValues {Curve {keys [<at 0 -0.2>,<at 1 1.2>]}}})
    Index 0 Param 2=1
    Index 0 Param 3=&Multiply_Float(
     A=&RandomRangeFloat(Minimum=0.8, Maximum=1)
     B=&FloatFromCurve(FloatCurve=DataInterface() {propertyValues {Curve {keys [<at 0 1>,<at 1 0.5>]}}})
    )
    Module.Param0WriteEnabled=true
    Module.Param1WriteEnabled=true
    Module.Param2WriteEnabled=true
    Module.Param3WriteEnabled=true
   )
   &ScaleColor(
    Scale Alpha=&Multiply_Float(
     A=&Multiply_Float(
      A=&Multiply_Float(A=1, B=&FNTimeOfDayScalar(Night Scale=1, Morning Scale=0.3, Evening Scale=0.4))
      B=&FloatFromCurve(FloatCurve=DataInterface() {propertyValues {Curve {keys [
           <at 0 0 RCIM_Cubic>
           <at 0.1912 0.9932>
           {arriveTangent=-1.2279, interpMode=RCIM_Cubic, leaveTangent=-1.2279, time=0.7879, value=1.0115}
           <at 1 0>
          ]}}})
     )
     B=&FloatFromCurve(
      FloatCurve=DataInterface() {propertyValues {Curve {keys [
          {interpMode=RCIM_Cubic, tangentMode=RCTM_User, time=0, value=0}
          {arriveTangent=0.37, interpMode=RCIM_Cubic, leaveTangent=0.37, tangentMode=RCTM_User, time=0.6, value=0.05}
          {arriveTangent=1.1869, interpMode=RCIM_Cubic, leaveTangent=1.1869, tangentMode=RCTM_User, time=0.9, value=0.5258}
          <at 1.2 0.8>
         ]}}}
      CurveIndex=$Emitter.Age
     )
    )
    ScaleA=true
    ScaleRGB=true
   )
   &ScaleMeshSize(Scale Factor=&Multiply_Vector(
     A=<1 1 1>
     B=&VectorFromFloat(Value=&FloatFromCurve(
       FloatCurve=DataInterface() {propertyValues {Curve {keys [
           {arriveTangent=0.2635, interpMode=RCIM_Cubic, leaveTangent=0.2635, tangentMode=RCTM_User, time=0, value=0.2}
           {arriveTangent=0.4932, interpMode=RCIM_Cubic, leaveTangent=0.4932, tangentMode=RCTM_User, time=1, value=0.5}
           {arriveTangent=0.447, interpMode=RCIM_Cubic, leaveTangent=0.447, tangentMode=RCTM_User, time=2, value=1.2}
          ]}}}
       CurveIndex=$System.Age
      ))
    ))
   &CameraOffset(Camera Offset Amount=20)
  ]
  rendererClasses [&NiagaraMeshRendererProperties]
  renderers [{rendererClass=&NiagaraMeshRendererProperties, rendererIndex=0}]
 }}
summary {
 systemName=NS_Emote_Spiral_Recharge
 emitters [{emitterName=Convergent_airflow_start, rendererClasses [&NiagaraMeshRendererProperties]}]
 userVariables []
}
